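# Material function that converts an axis-angle vector to a quaternion and a quaternion back to axis-angle, as two separate outputs.
# Both forms are Vector4, with the axis or vector part in RGB and the radian angle or scalar in alpha.
# Axis-to-quat uses half-angle sine and cosine without normalizing the axis.
# Quat-to-axis doubles arccos(w) and divides XYZ by sqrt((1-w)*w), not the standard sqrt(1-w*w), with no zero guard.
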MaterialFunction /Engine/Functions/Engine_MaterialFunctions02/Math/QuatAxisAngleConversions
graph {
 properties {description=Convert a quat to an angle axis and vice versa.}
 nodes [
  n1 = FunctionInput() {inputName=AxisAngle, inputType=FunctionInput_Vector4}
  n2 = ComponentMask(None=@n1) {b=true, g=true, r=true}
  n3 = Multiply(A=MaterialFunctionCall(Float4=@n1) {materialFunction=&BreakOutFloat4Components}.A) {constB=0.5}
  n4 = Sine(None=@n3) {period=6.2832}
  FunctionOutput(None=AppendVector(A=Multiply(A=@n2, B=@n4), B=Cosine(None=@n3) {period=6.2832})) {outputName=Axis Angle As Quat}
  n5 = FunctionInput() {inputName=Quaternion, inputType=FunctionInput_Vector4}
  n6 = ComponentMask(None=@n5) {b=true, g=true, r=true}
  n7 = ComponentMask(None=@n5) {a=true}
  n8 = SquareRoot(None=Multiply(A=OneMinus(None=@n7), B=@n7))
  n9 = Arccosine!(None=@n7)
  FunctionOutput(None=AppendVector(A=Divide(A=@n6, B=@n8), B=Multiply(A=@n9) {constB=2})) {outputName=Quat as Axis Angle}
 ]
}
