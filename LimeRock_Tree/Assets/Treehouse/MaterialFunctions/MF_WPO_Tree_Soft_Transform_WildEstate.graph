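# Material function that computes wind-driven world position offset for trees, layering trunk bend, branch and leaf motion.
# Sine oscillators blend calm and storm speeds by a storm mask, which defaults from MF_WindActorData alongside wind intensity.
# Branches use quaternions and pivots sampled from textures via UV1; leaf pivots and directions are packed into UV1/UV2.
# An IsTrunk switch omits extra leaf offsets; outputs include Offset, leaf pivot, overall bend and debug.
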MaterialFunction /LimeRock_Tree/Assets/Treehouse/MaterialFunctions/MF_WPO_Tree_Soft_Transform_WildEstate
graph {nodes [
  n1 = NamedRerouteUsage()
  n2 = NamedRerouteUsage()
  n3 = Add(A=Time(), B=@n2)
  n4 = Multiply(A=@n2) {constB=0.01}
  n5 = Multiply(A=@n3, B=Add(A=ScalarParameter() {defaultValue=0.123, group=Animation, parameterName=Trunk_Anim_Speed}, B=@n4))
  n6 = Multiply(A=@n3, B=Add(A=@n4, B=ScalarParameter() {defaultValue=0.5, group=Animation, parameterName=Trunk_Anim_Speed_Storm}))
  n7 = Add(
   A=LinearInterpolate(A=Sine(None=@n5), B=Sine(None=@n6), Alpha=NamedRerouteUsage())
   B=ScalarParameter() {defaultValue=0.65, group=Animation, parameterName=Trunk_Anim_Bias}
  )
  n8 = Multiply(A=@n7) {constB=-1}
  n9 = NamedRerouteUsage()
  n10 = Multiply(A=@n7, B=ScalarParameter() {defaultValue=100, group=Animation, parameterName=Trunk_Anim_DirectionOffset_Branches})
  n11 = NamedRerouteUsage()
  n12 = ComponentMask(None=NamedRerouteUsage()) {a=true}
  n13 = Frac(None=@n12)
  n14 = Round(None=@n13)
  n15 = Add(
   A=Multiply(A=Multiply(A=@n8, B=@n9), B=ScalarParameter() {defaultValue=150, group=Animation, parameterName=Trunk_Anim_DirectionOffset})
   B=Multiply(A=Multiply(A=@n10, B=@n11), B=Saturate(None=@n14))
  )
  n16 = NamedRerouteDeclaration(None=Multiply(
    A=MaterialFunctionCall() {materialFunction=&ObjectScale}.Scale Z
    B=Multiply(A=@n1, B=@n15)
   )) {name=OverallBend, nodeColor=<rgba 0.5938 0 1 1>}
  Add(A=NamedRerouteUsage(), B=@n16)
  n17 = LocalPosition()
  n18 = NamedRerouteUsage()
  n19 = Normalize(VectorInput=Transform(None=Subtract(A=@n17.XYZ, B=@n18)) {transformSourceType=TRANSFORMSOURCE_Local})
  n20 = DotProduct(A=Constant3Vector() {constant=<rgba 0 0 1 0>}, B=@n19)
  Add(A=@n20)
  n21 = LocalPosition() {includedOffsets=ExcludeOffsets}
  Add(A=@n21.XYZ, B=NamedRerouteUsage())
  Ceil(None=@n12)
  Constant3Vector()
  Constant3Vector()
  Constant3Vector() {desc=WindDirection, constant=<rgba 1 0 0 0>}
  Constant()
  Constant()
  Constant() {r=4}
  n22 = Add(A=NamedRerouteUsage(), B=NamedRerouteUsage())
  n23 = Add(A=Time(), B=@n22)
  n24 = Multiply(A=@n22) {constB=0.01}
  n25 = ScalarParameter() {defaultValue=0.5257, group=Animation, parameterName=Test_Anim_Speed}
  n26 = Add(A=@n24, B=@n25)
  n27 = ScalarParameter() {defaultValue=1, group=Animation, parameterName=Test_Anim_Speed_Storm}
  n28 = Add(A=@n24, B=@n27)
  n29 = LinearInterpolate(A=Sine(None=Multiply(A=@n23, B=@n26)), B=Sine(None=Multiply(A=@n23, B=@n28)), Alpha=NamedRerouteUsage())
  Divide(A=Add(A=@n29))
  n30 = LinearInterpolate(A=Sine(None=Multiply(A=@n23, B=@n25)), B=Sine(None=Multiply(A=@n23, B=@n27)), Alpha=NamedRerouteUsage())
  Divide(A=Add(A=@n30))
  Floor(None=@n12)
  n31 = NamedRerouteUsage()
  n32 = Add(A=Multiply(A=VertexNormalWS()) {constB=0.5}, B=@n31)
  n33 = DotProduct(A=@n31, B=@n19)
  n34 = Multiply(A=Sign(None=@n33), B=Power(Base=@n33, Exp=ScalarParameter() {defaultValue=1.107, parameterName=Leaf_Offset_Power}))
  n35 = Multiply(A=Add(A=@n29) {constB=0.5}, B=Multiply(A=ScalarParameter() {defaultValue=5, parameterName=Leaf_Offset_intenity}, B=Multiply(A=@n32, B=@n34)))
  n36 = ScalarParameter() {defaultValue=5, parameterName=Leaf_Offset_intenity2}
  n37 = Multiply(A=Multiply(A=@n31, B=Multiply(A=@n36, B=@n20)), B=@n30)
  n38 = NamedRerouteUsage()
  n39 = StaticSwitchParameter(True=@n38, False=Add(A=@n38, B=Add(A=@n35, B=@n37))) {parameterName=IsTrunk}
  FunctionOutput(None=Add(A=NamedRerouteUsage(), B=Transform(None=@n39) {transformSourceType=TRANSFORMSOURCE_Local})) {outputName=Offset}
  n40 = TextureCoordinate() {coordinateIndex=1}
  n41 = TextureCoordinate() {coordinateIndex=2}
  n42 = ComponentMask(None=@n41) {r=true}
  n43 = ComponentMask(None=@n41) {g=true}
  n44 = Frac(None=AppendVector(A=ComponentMask(None=@n40) {r=true}, B=AppendVector(A=@n42, B=@n43)))
  n45 = Multiply(A=@n44) {constB=12000}
  n46 = Constant3Vector() {constant=<rgba 6000 6000 3000 0>}
  FunctionOutput(None=NamedRerouteDeclaration(None=Subtract(A=@n45, B=@n46) {constB=6000}) {name=LeafPivot Position, nodeColor=<rgba 0.0781 1 0 1>}) {outputName=LeafPivot_LocalSpace, sortPriority=10}
  FunctionOutput(None=@n16) {outputName=OveraAllBend, sortPriority=100}
  n47 = Add(A=ComponentMask(None=TextureCoordinate() {coordinateIndex=1}) {g=true}) {constB=16}
  n48 = NamedRerouteDeclaration(None=Saturate(None=Divide(A=@n47) {constB=32, desc=quantiazied fall off with 32 steps})) {name=BranchFalloff, nodeColor=<rgba 0.875 0 1 1>}
  n49 = Constant2Vector() {g=-1, r=-1}
  n50 = ComponentMask(None=WorldPosition().XYZ) {b=true, r=true}
  n51 = Frac(None=Divide(A=Multiply(A=@n49, B=@n50)) {constB=32})
  FunctionOutput(None=MaterialFunctionCall(Number=VertexInterpolator(VS=@n48).PS, UVs=@n51) {materialFunction=&DebugScalarValues}.Result) {outputName=Debug}
  MaterialFunctionCall!(Vector3, MaximumNumberOfDigits, UVs=@n51, DebugTextLocation RG_UpperRight BA_LowerLeft, Component Spacing) {materialFunction=/Engine/Functions/Engine_MaterialFunctions02/Utility/DebugFloat3Values}
  MaterialFunctionCall!(Vector4=NamedRerouteUsage(), MaximumNumberOfDigits, UVs=@n51, DebugTextLocation RG_UpperRight BA_LowerLeft, Component Spacing) {materialFunction=/Engine/Functions/Engine_MaterialFunctions02/Utility/DebugFloat4Values}
  Multiply(A=NamedRerouteUsage(), B=NamedRerouteUsage())
  n52 = MaterialFunctionCall!(
   Position=NamedRerouteUsage()
   IsTrees=StaticBool() {value=true}
   IsGrass
   IsMediumPlant
   IsMediumPlant_2024Anim
   IsMediumPlant2ndLevel
   MediumPlant_BasePivotPostion
  ) {materialFunction=/Game/Environments/Asteria/Foliage/Wind/MF_WindActorData}
  NamedRerouteDeclaration(None=@n52.WindIntensitySingleSample) {name=WindIntensity, nodeColor=<rgba 0 1 0.3438 1>}
  n53 = ScalarParameter() {defaultValue=8671.8857, parameterName=Frequency}
  n54 = NamedRerouteUsage()
  NamedRerouteDeclaration(None=MaterialFunctionCall(Position=Multiply(A=@n53, B=@n54), IsTrees=StaticBool() {value=true}) {materialFunction=&MF_WindActorData}.WindIntensitySingleSample) {name=WindIntensity_perleaf, nodeColor=<rgba 0 1 0.3438 1>}
  n55 = TextureSampleParameter2D(UVs=ComponentMask(None=TextureCoordinate() {coordinateIndex=1}) {g=true}) {
   group=Animation
   parameterName=Anim_Rotation
   samplerType=SAMPLERTYPE_LinearColor
   texture=/Game/Environments/Asteria/Foliage/Trees/MasterMaterials/Prototype2/Rotation
  }
  NamedRerouteDeclaration(None=@n55.RGBA) {name=Quaternion, nodeColor=<rgba 0 0.625 1 1>}
  NamedRerouteDeclaration(None=Multiply(A=@n55.RGBA, B=Constant4Vector() {constant=<rgba -1 -1 -1 1>}.RGBA) {constB=-1}) {name=Quaternion_Inverse, nodeColor=<rgba 0.9453 1 0 1>}
  NamedRerouteDeclaration(None=Multiply(
    A=VertexColor() {desc=RedVertex Color Per Leaf Random}.R
    B=ScalarParameter() {defaultValue=0.75, group=Animation, parameterName=Leaf_Anim_Randomness}
   ) {constB=33.33}) {name=LeafMotionRandomness, nodeColor=<rgba 1 0.8906 0 1>}
  n56 = TextureCoordinate() {coordinateIndex=2}
  n57 = AppendVector(A=ComponentMask(None=TextureCoordinate() {coordinateIndex=1}) {r=true}, B=ComponentMask(None=@n56) {r=true})
  n58 = ComponentMask(None=@n56) {g=true}
  n59 = Divide(A=Floor(None=AppendVector(A=@n57, B=@n58) {desc=LeafDirectionVector})) {constB=32}
  NamedRerouteDeclaration(None=@n59) {name=LeafDirectionVector, nodeColor=<rgba 0.2422 0 1 1>}
  n60 = TextureSampleParameter2D(UVs=ComponentMask(None=TextureCoordinate() {coordinateIndex=1}) {g=true}) {
   group=Animation
   parameterName=Anim_Position
   samplerType=SAMPLERTYPE_LinearColor
   texture=/Game/Environments/Asteria/Foliage/Trees/MasterMaterials/Prototype2/Position
  }
  NamedRerouteDeclaration(None=@n60.RGB) {name=Branch Pivot, nodeColor=<rgba 0 0.7187 1 1>}
  NamedRerouteDeclaration(None=@n60.A) {name=Random Value Per Branch, nodeColor=<rgba 1 0.9609 0 1>}
  n61 = Add(A=NamedRerouteUsage(), B=NamedRerouteUsage())
  n62 = Add(A=Time(), B=@n61)
  n63 = Multiply(A=@n61) {constB=0.01}
  n64 = Add(A=@n63, B=ScalarParameter() {defaultValue=0.9, group=Animation, parameterName=Leaf_Anim_Speed})
  n65 = Add(A=ScalarParameter() {defaultValue=3, group=Animation, parameterName=Leaf_Anim_Speed_Storm}, B=@n63)
  n66 = LinearInterpolate(A=Sine(None=Multiply(A=@n62, B=@n64)), B=Sine(None=Multiply(A=@n62, B=@n65)), Alpha=NamedRerouteUsage())
  n67 = Divide(A=Add(A=@n66))
  n68 = ScalarParameter() {defaultValue=0.25, group=Animation, parameterName=Leaf_Anim_Bias}
  n69 = NamedRerouteUsage()
  n70 = Constant3Vector() {constant=<rgba 0 0 -1 0>}
  n71 = NamedRerouteUsage()
  n72 = CrossProduct(A=@n69, B=@n71)
  n73 = StaticSwitchParameter(
   True=Normalize(VectorInput=CrossProduct(A=@n69, B=@n70))
   False=StaticSwitchParameter(True=Normalize(VectorInput=@n72), False=@n69) {group=Animation, parameterName=LeafRotation_InWindDirection}
  ) {group=Animation, parameterName=LeafRotation_InUpDown}
  n74 = Add(A=@n66, B=@n68)
  n75 = StaticSwitchParameter(
   True=LinearInterpolate(A=NamedRerouteUsage(), B=NamedRerouteUsage(), Alpha=Add(A=@n67, B=@n68))
   False=MaterialFunctionCall!(Quaternion, AxisAngle=AppendVector(A=@n73, B=@n74)) {materialFunction=/Engine/Functions/Engine_MaterialFunctions02/Math/QuatAxisAngleConversions}.Axis Angle As Quat
  ) {desc=False Is Higher quality but True could look good on some trees, group=Animation, parameterName=Leave use Branch Quaterninon}
  n76 = LocalPosition() {includedOffsets=ExcludeOffsets}
  n77 = NamedRerouteUsage()
  n78 = Abs(None=DotProduct(A=Subtract(A=@n76.XYZ, B=@n77), B=NamedRerouteUsage()))
  n79 = ScalarParameter() {defaultValue=5, group=Animation, parameterName=Leaf_Gradient_Falloff_Distance}
  n80 = ScalarParameter() {defaultValue=0.37, group=Animation, parameterName=Leaf_Anim_MaxRotation}
  n81 = Multiply(
   A=Power(Base=Divide(A=@n78, B=@n79), Exp=ScalarParameter() {defaultValue=1, group=Animation, parameterName=Leaf_Gradient_Falloff_Power})
   B=@n80
  )
  n82 = Normalize(VectorInput=LinearInterpolate(
    A=Constant4Vector() {constant=<rgba 0 0 0 1>}.RGBA
    B=@n75
    Alpha=StaticSwitchParameter(True=@n81, False=@n80) {group=Animation, parameterName="Leaf_Gradient_Falloff(EXPENSIVE)"}
   ))
  n83 = ComponentMask(None=@n82) {b=true, g=true, r=true}
  n84 = LocalPosition() {includedOffsets=ExcludeOffsets}
  n85 = Subtract(A=@n84.XYZ, B=NamedRerouteUsage())
  n86 = Multiply(A=CrossProduct(A=@n83, B=@n85)) {constB=2}
  n87 = CrossProduct(A=@n83, B=@n86)
  n88 = Add(A=Multiply(A=@n86, B=ComponentMask(None=@n82) {a=true}), B=@n85)
  NamedRerouteDeclaration(None=Subtract(A=Add(A=@n87, B=@n88), B=@n85)) {name=LeafMovement, nodeColor=<rgba 0 1 0.5547 1>}
  n89 = ObjectRadius()
  n90 = MaterialFunctionCall() {materialFunction=&ObjectScale}
  n91 = Divide(
   A=Subtract(A=LocalPosition() {includedOffsets=ExcludeOffsets}.Z, B=ScalarParameter() {group=Animation, parameterName=Trunk_Anim_BendOffset})
   B=Multiply(A=Divide(A=@n89, B=@n90.Scale Z)) {constB=2}
  )
  n92 = Multiply(A=@n91, B=ScalarParameter() {defaultValue=0.5, group=Animation, parameterName=Trunk_Anim_BendFalloff})
  NamedRerouteDeclaration(None=Saturate(None=Multiply(A=@n91, B=@n92))) {name=TrunkGradient, nodeColor=<rgba 0.1719 0 1 1>}
  NamedRerouteDeclaration(None=Constant3Vector() {desc=WindDirection, constant=<rgba -0.6963 0.6963 0.1741 0>}) {name=WindDIrection, nodeColor=<rgba 0 1 0.2734 1>}
  n93 = NamedRerouteUsage()
  n94 = NamedRerouteUsage()
  n95 = NamedRerouteUsage()
  n96 = Add(A=Time(), B=Add(A=NamedRerouteUsage(), B=@n95))
  n97 = Multiply(A=@n95) {constB=0.01}
  n98 = ScalarParameter() {defaultValue=0.25, group=Animation, parameterName=Branches_Anim_Speed}
  n99 = Sine(None=Multiply(A=@n96, B=Add(A=@n97, B=@n98)))
  n100 = ScalarParameter() {defaultValue=1, group=Animation, parameterName=Branches_Anim_Speed_Storm}
  n101 = Sine(None=Multiply(A=@n96, B=Add(A=@n97, B=@n100)))
  n102 = NamedRerouteUsage()
  n103 = Divide(A=Add(A=LinearInterpolate(A=@n99, B=@n101, Alpha=@n102)))
  n104 = NamedRerouteUsage()
  n105 = ScalarParameter() {defaultValue=0.2, group=Animation, parameterName=Branches_Anim_MaxRotation}
  n106 = Normalize(VectorInput=LinearInterpolate(A=Constant4Vector() {constant=<rgba 0 0 0 1>}.RGBA, B=LinearInterpolate(A=@n93, B=@n94, Alpha=@n103), Alpha=Multiply(A=@n104, B=@n105)))
  n107 = ComponentMask(None=@n106) {b=true, g=true, r=true}
  n108 = Subtract(A=@n21.XYZ, B=NamedRerouteUsage())
  n109 = Multiply(A=CrossProduct(A=@n107, B=@n108)) {constB=2}
  n110 = CrossProduct(A=@n107, B=@n109)
  n111 = Add(A=Multiply(A=@n109, B=ComponentMask(None=@n106) {a=true}), B=@n108)
  NamedRerouteDeclaration(None=Subtract(A=Add(A=@n110, B=@n111), B=@n108)) {name=BranchMovement, nodeColor=<rgba 1 0 0.0469 1>}
  n112 = MaterialFunctionCall(Float3=NamedRerouteUsage()) {materialFunction=&BreakOutFloat3Components}
  n113 = Divide(A=Add(A=@n112.R, B=@n112.R)) {constB=100}
  NamedRerouteDeclaration(None=Multiply(A=Frac(None=@n113)) {constB=23.33}) {name=RandomPerInstance, nodeColor=<rgba 0 1 0.5078 1>}
  n114 = Constant3Vector()
  NamedRerouteDeclaration(None=StaticSwitch(
    True=ActorPositionWS()
    False=TransformPosition(None=@n114) {transformSourceType=TRANSFORMPOSSOURCE_Instance, transformType=TRANSFORMPOSSOURCE_World}
    Value=NamedRerouteUsage()
   )) {name=InstancePosition, nodeColor=<rgba 0 1 0.5781 1>}
  NamedRerouteDeclaration(None=FunctionInput(Preview=@n52.StormMask) {bUsePreviewValueAsDefault=true, inputName=StormMask, inputType=FunctionInput_Scalar, sortPriority=797}) {name=StormMask, nodeColor=<rgba 0.7344 1 0 1>}
  NamedRerouteDeclaration(None=FunctionInput(Preview=StaticBool()) {bUsePreviewValueAsDefault=true, inputName=TreeAddOnOverallBend, inputType=FunctionInput_StaticBool, sortPriority=798}) {name=IsTreeAddon, nodeColor=<rgba 0.4531 0 1 1>}
  NamedRerouteUsage()
  NamedRerouteUsage()
  NamedRerouteUsage()
  NamedRerouteUsage()
  n115 = Transform(None=NamedRerouteUsage()) {transformSourceType=TRANSFORMSOURCE_Local}
  n116 = NamedRerouteUsage()
  n117 = CrossProduct(A=@n115, B=@n116)
  Normalize(VectorInput=CrossProduct(A=@n115, B=@n117)) {desc=Leaf roationAxis}
  Normalize(VectorInput=@n59)
  OneMinus(None=@n33)
  PerInstanceCustomData3Vector!(DefaultValue)
  n118 = Normalize(VectorInput=@n117) {desc=Leaf roationAxis}
  n119 = Add(A=Time(), B=NamedRerouteUsage())
  n120 = ScalarParameter() {defaultValue=0.5, group=Animation, parameterName=Leaves_Speed}
  n121 = ScalarParameter() {defaultValue=2, group=Animation, parameterName=Leaves_Speed_Storm}
  n122 = LinearInterpolate(A=Frac(None=Multiply(A=@n119, B=@n120)), B=Frac(None=Multiply(A=@n119, B=@n121)), Alpha=NamedRerouteUsage())
  n123 = NamedRerouteUsage()
  n124 = NamedRerouteUsage()
  RotateAboutAxis(
   NormalizedRotationAxis=@n118
   RotationAngle=Multiply(A=Sine(None=@n122), B=ScalarParameter() {defaultValue=0.05, group=Animation, parameterName=Leaves_MaxRotation})
   PivotPoint=StaticSwitchParameter(
    True=TransformPosition(None=@n123) {transformType=TRANSFORMPOSSOURCE_TranslatedWorld}
    False=TransformPosition(None=@n124) {transformType=TRANSFORMPOSSOURCE_World}
   ) {parameterName=OptimizeCameraRelativePos}
   Position=StaticSwitchParameter(
    True=WorldPosition() {worldPositionShaderOffset=WPT_CameraRelativeNoOffsets}.XYZ
    False=WorldPosition() {worldPositionShaderOffset=WPT_ExcludeAllShaderOffsets}.XYZ
   ) {parameterName=OptimizeCameraRelativePos}
  )
  ShaderStageSwitch(PixelShader=VertexInterpolator(VS=@n84.XYZ).PS, VertexShader=@n84.XYZ)
  n125 = NamedRerouteUsage()
  StaticSwitchParameter(True=Normalize(VectorInput=CrossProduct(A=@n125, B=@n116)) {desc=Leaf roationAxis}, False=@n118) {parameterName=OptimizeCameraRelativePos}
  Subtract(A=NamedRerouteUsage(), B=@n13)
  n126 = StaticSwitchParameter(True=WorldPosition() {worldPositionShaderOffset=WPT_CameraRelativeNoOffsets}.XYZ, False=WorldPosition().XYZ) {parameterName=OptimizeCameraRelativePos}
  n127 = Constant3Vector()
  n128 = StaticSwitch(
   True=Add(A=@n126, B=FunctionInput(Preview=@n127) {bUsePreviewValueAsDefault=true, inputName=OptionalWPO, sortPriority=798})
   False=@n126
   Value=NamedRerouteUsage()
  )
  Subtract(
   A=ComponentMask(None=NamedRerouteDeclaration(None=@n128) {name=WPO_plus_Offset, nodeColor=<rgba 0 1 0.8594 1>}) {b=true}
   B=ComponentMask(None=NamedRerouteUsage()) {b=true}
  )
  Subtract(A=Multiply(A=@n44) {constB=32752}) {constB=16376}
  TextureCoordinate() {uTiling=-6, vTiling=-6}
  Transform(None=@n71) {transformSourceType=TRANSFORMSOURCE_World, transformType=TRANSFORM_Local}
 ]}
