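# LimeRock tree material function that fakes unlit shading by feeding time-of-day tinted base color into emissive.
# It takes Material Attributes and multiplies BaseColor by the Color_Time_of_Day result, blended from Morning, Day, Evening and Night color parameters.
# A brightness parameter scales that tint into EmissiveColor, while a second dims the original BaseColor.
# All other attributes pass through unchanged.
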MaterialFunction /LimeRock_Tree/Assets/Treehouse/MaterialFunctions/MF_LimeRock_UnlitShading
graph {nodes [
  n1 = GetMaterialAttributes(None=FunctionInput() {inputType=FunctionInput_MaterialAttributes}) {attributeGetTypes [#1]}
  n2 = Multiply(
   A=@n1.BaseColor
   B=MaterialFunctionCall(
    Color_Morning=VectorParameter() {parameterName=Morning, defaultValue=<rgba 0.6615 0.599 0.4331 1>}.RGB
    Color_Day=VectorParameter() {parameterName=Day, defaultValue=<rgba 0.8021 0.8021 0.8021 1>}.RGB
    Color_Evening=VectorParameter() {parameterName=Evening, defaultValue=<rgba 0.6042 0.54 0.4278 1>}.RGB
    Color_Night=VectorParameter() {parameterName=Night, defaultValue=<rgba 0.2996 0.4717 1 1>}.RGB
   ) {materialFunction=&Color_Time_of_Day}.Result
  )
  n3 = ScalarParameter() {defaultValue=1.3, parameterName=BrightnessUnlitEmissive}
  n4 = ScalarParameter() {defaultValue=0.5, parameterName=BrightnessBaseColor}
  FunctionOutput(None=SetMaterialAttributes!(MaterialAttributes=@n1.MaterialAttributes, EmissiveColor=Multiply(A=@n2, B=@n3), BaseColor=Multiply(A=@n1.BaseColor, B=@n4) {constB=0.5}) {attributeSetTypes [#2, #1]})
 ]}
